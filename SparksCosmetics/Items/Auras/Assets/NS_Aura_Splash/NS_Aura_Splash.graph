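# Niagara system for the Splash character aura cosmetic: water drips, splashes, splatter, decals and tumbling beach balls around a skeletal mesh.
# Drips and splash sprites spawn between paired bones, many pushed outward from bone 7, with drips culled by bone speed.
# Beach balls launch from bone 7, with a random subset able to bounce.
# Front-end, large-body and time-of-day inputs tune the look; CausticGlows and Beachballs001 are disabled.
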NiagaraSystem /SparksCosmetics/Items/Auras/Assets/NS_Aura_Splash/NS_Aura_Splash
dependencies {
 usedDataInterfaces [
  &NiagaraDataInterfaceArrayInt32
  &NiagaraDataInterfaceCurve
  &NiagaraDataInterfaceArrayFloat
  &NiagaraDataInterfaceColorCurve
  &NiagaraDataInterfaceVector2DCurve
  &NiagaraDataInterfaceArrayFloat2
  &NiagaraDataInterfaceSkeletalMesh
 ]
 usedDynamicInputs [
  &MakeVector
  &Multiply_Vector2
  &Vector2DFromFloat
  &MakeIntFromFloat
  &Frac_Float
 ]
 usedModules [
  &self:SystemSpawnScript.SystemScriptSource.SystemScriptGraph.NiagaraNodeAssignment_0.SetVariables_#1
  &self:SystemSpawnScript.SystemScriptSource.SystemScriptGraph.NiagaraNodeAssignment_2.SetVariables_#2
  &SystemState
  &self:SystemSpawnScript.SystemScriptSource.SystemScriptGraph.NiagaraNodeAssignment_3.SetVariables_#3
 ]
}
emitters {
 Beachballs {
  emitterUpdateScript [
   &EmitterState()
   &SpawnRate(Spawn Probability=0.5, SpawnRate=2, Use Spawn Probability=true)
  ]
  particleSpawnScript [
   /SetVariables(Particles.UseCollision:NiagaraBool=&Boolean_AndOperation(Bool1=&RandomBool(), Bool2=&RandomBool())) {isSetParameters=true}
   /SetVariables(Particles.UseCollision=true) {enabled=false, isSetParameters=true}
   &InitializeParticle(
    Color Mode=::NewEnumerator0
    Lifetime=&MakeCustomFloatFromBool(Bool=$Particles.UseCollision, True Float=2, False Float=&RandomRangeFloat(Minimum=0.75, Maximum=1.25))
    Mesh Scale Mode=::NewEnumerator1
    Mesh Uniform Scale Max=0.5
    Mesh Uniform Scale Min=0.3
    Position Offset=&RandomVector(Vector Scale=20)
    UsePositionOffset=true
    Write Material Random=true
   )
   &SkeletalMeshLocation(
    Bone / Socket Index=7
    Bone Sampling Mode=::NewEnumerator4
    MaskSpawnGroup=false
    Orientation Sampling=::NewEnumerator1
    Skeletal Mesh=$System.Skeletal Mesh
   )
   &InitialMeshOrientation(Use Rotation Vector=false)
   &AddVelocity(Velocity=&Multiply_Vector(
     A=&MakeCustomVectorFromBool(Bool=$Particles.UseCollision, True Vector=<1 1 1>, False Vector=<2 1 1>)
     B=&self:ClampVectorByLength(
      Vector:Vector3f=/Niagara/DynamicInputs/Math/LerpMultipleVectors(
       Alpha:NiagaraFloat=&RandomRangeFloat(Minimum=0, Maximum=1)
       V 1:Vector3f=<-75 -50 150>
       V 2:Vector3f=<0 -40 200>
       V 3:Vector3f=<75 -50 150>
       V 4:Vector3f=?
       V 5:Vector3f=?
       Vector Count:ENiagara_MultipleLerpCount=::NewEnumerator0
      )
      Min:NiagaraFloat=200
      Max:NiagaraFloat=300
     )
    ))
   /SetVariables(Particles.EndLength=0.25) {isSetParameters=true}
  ]
  particleUpdateScript [
   &ParticleState()
   /SetVariables(Transient.FindEnding:NiagaraFloat=&Max_Float(
     A:NiagaraFloat=&Subtract_Float(A=$Particles.Age, B=&Subtract_Float(A=$Particles.Lifetime, B=$Particles.EndLength))
     B:NiagaraFloat=0
    )) {isSetParameters=true}
   &ScaleMeshSize(Scale Factor=&VectorFromFloat(Value=&FloatFromCurve(
      FloatCurve=DataInterface() {propertyValues {Curve {keys [
          {arriveTangent=32.9812, interpMode=RCIM_Cubic, leaveTangent=32.9812, tangentMode=RCTM_User, time=0, value=0}
          {arriveTangentWeight=0.1389, interpMode=RCIM_Cubic, leaveTangentWeight=0.2778, tangentMode=RCTM_User, time=0.2, value=1}
         ]}}}
      Scale Curve=&FloatFromCurve(
       FloatCurve=DataInterface() {propertyValues {Curve {keys [
           {arriveTangentWeight=0.2778, interpMode=RCIM_Cubic, leaveTangentWeight=0.1389, tangentMode=RCTM_User, time=0, value=1}
           {arriveTangent=-7.0693, interpMode=RCIM_Cubic, leaveTangent=-7.0693, tangentMode=RCTM_User, time=1, value=0}
          ]}}}
       CurveIndex=&Divide_Float(A=$Transient.FindEnding, B=$Particles.EndLength)
      )
     )))
   &ScaleMeshSize(Scale Factor=&VectorFromFloat(Value=&FloatFromCurve(FloatCurve=DataInterface() {propertyValues {Curve {keys [
          {arriveTangent=32.9812, interpMode=RCIM_Cubic, leaveTangent=32.9812, tangentMode=RCTM_User, time=0, value=0}
          {arriveTangentWeight=0.1389, interpMode=RCIM_Cubic, leaveTangentWeight=0.2778, tangentMode=RCTM_User, time=0.2, value=1}
          {arriveTangentWeight=0.2778, interpMode=RCIM_Cubic, leaveTangentWeight=0.1389, tangentMode=RCTM_User, time=0.8, value=1}
          {arriveTangent=-34.5131, interpMode=RCIM_Cubic, leaveTangent=-34.5131, tangentMode=RCTM_User, time=1, value=0}
         ]}}}))) {enabled=false, module=ScaleMeshSize001}
   &GravityForce(Gravity=<0 0 -300>)
   &Collision(
    Age Colliding Particles=false
    Collision Enabled=&Boolean_AndOperation(Bool1=$Particles.UseCollision, Bool2=&SetBoolByFloatComparison(A=$Particles.Age, B=1))
    KillOnCollision=false
    Mesh Dimensions:Vector3f=<60 60 60>
    Module.Control Rest State Via Surface Normal=false
    Module.EnableMaxCollisionCount=false
    Module.Randomize Collision Normal Vector=false
    Particle Radius Scale:NiagaraFloat=&MakeFloatFromVector(Vector=$Particles.Scale, Channel=::NewEnumerator0)
    Radius Calculation Type=::NewEnumerator1
    Restitution=1
   )
   &Drag(
    Drag=&MakeCustomFloatFromBool(
     Bool=$Particles.UseCollision
     True Float=&MakeCustomFloatFromBool(Bool=$Particles.HasCollided, True Float=1, False Float=0.1)
     False Float=1
    )
    Use Linear Drag=true
    Use Rotational Drag=false
   )
   &SolveForcesAndVelocity(Clamp Velocity=false, Limit Acceleration=false)
   /SetVariables(Transient.VelocityScale:NiagaraFloat=&RemapRange(
     Input Value To Remap=&VectorLength(Vector=$Particles.Velocity)
     Input Min=75
     Input Max=125
     Output Min=0.25
     Output Max=1
    )) {isSetParameters=true, module=SetVariables_#4}
   &UpdateMeshOrientation(Rotation Rate=$Transient.VelocityScale, Rotation Vector=&RandomVector(Vector Scale=1))
   &Color(Color=&ColorFromCurve(
     ColorCurve=DataInterface() {propertyValues {
       AlphaCurve {keys [<at 0 1 RCIM_Cubic>]}
       BlueCurve {keys [<at 0 0.9943>, {arriveTangent=-0.9943, interpMode=RCIM_Cubic, leaveTangent=-0.9943, time=0.02, value=0}, <at 0.98 1>, <at 1 0 RCIM_Cubic>]}
       GreenCurve {keys [<at 0 0.9943>, {arriveTangent=0.0057, interpMode=RCIM_Cubic, leaveTangent=0.0057, time=0.02, value=0}, <at 0.98 0>, <at 1 1 RCIM_Cubic>]}
       RedCurve {keys [<at 0 0.9943>, {arriveTangent=-0.9943, interpMode=RCIM_Cubic, leaveTangent=-0.9943, time=0.02, value=1}, <at 0.98 0>, <at 1 0 RCIM_Cubic>]}
      }}
     CurveIndex=$Transient.VelocityScale
    )) {enabled=false}
   /SetVariables(Particles.CameraOffset:NiagaraFloat=&FloatFromCurve(FloatCurve=DataInterface() {propertyValues {Curve {keys [<at 1 1>,<at 1.5 0>]}}}, CurveIndex=$Particles.Age, Scale Curve=-50)) {isSetParameters=true}
  ]
  rendererClasses [&NiagaraSpriteRendererProperties, &NiagaraMeshRendererProperties]
  renderers [
   {rendererClass=&NiagaraSpriteRendererProperties}
   {rendererClass=&NiagaraMeshRendererProperties, rendererIndex=1}
  ]
 }
 Beachballs001 {
  bEnabled=false
  emitterUpdateScript [
   &EmitterState()
   &SpawnRate(Spawn Probability=0.5, SpawnRate=10, Use Spawn Probability=true)
  ]
  particleSpawnScript [
   &InitializeParticle(
    Color Mode=::NewEnumerator0
    Lifetime=&Lerp_Float(
     A=0.75
     B=2
     Alpha=&FloatFromCurve(
      FloatCurve=DataInterface() {propertyValues {Curve {keys [
          {arriveTangent=1.7336, interpMode=RCIM_Cubic, leaveTangent=1.7336, tangentMode=RCTM_User, time=0, value=0}
          {arriveTangent=1, interpMode=RCIM_Cubic, leaveTangent=1, time=0.9, value=0.6}
          {arriveTangent=5.784, tangentMode=RCTM_User, time=1, value=1}
         ]}}}
      Scale Curve=&RandomRangeFloat(Minimum=0, Maximum=1)
     )
    )
    Mesh Scale Mode=::NewEnumerator1
    Mesh Uniform Scale Max=0.5
    Mesh Uniform Scale Min=0.2
    Position Offset=&RandomVector(Vector Scale=20)
    UsePositionOffset=true
    Write Material Random=true
   )
   &SkeletalMeshLocation(
    Bone / Socket Index=7
    Bone Sampling Mode=::NewEnumerator4
    MaskSpawnGroup=false
    Orientation Sampling=::NewEnumerator1
    Sampled Position Offset=<0 -25 0>
    Skeletal Mesh=$System.Skeletal Mesh
   )
   &InitialMeshOrientation(Use Rotation Vector=false)
   &AddVelocity(Velocity=&self:ClampVectorByLength(
     Vector:Vector3f=&LerpMultipleVectors(
      Alpha:NiagaraFloat=&RandomRangeFloat(Minimum=0, Maximum=1)
      V 1:Vector3f=<-75 -50 150>
      V 2:Vector3f=<0 -40 200>
      V 3:Vector3f=<75 -50 150>
      V 4:Vector3f=?
      V 5:Vector3f=?
      Vector Count:ENiagara_MultipleLerpCount=::NewEnumerator0
     )
     Min:NiagaraFloat=200
     Max:NiagaraFloat=300
    ))
  ]
  particleUpdateScript [
   &ParticleState()
   &ScaleMeshSize(Scale Factor=&VectorFromFloat(Value=&FloatFromCurve(
      FloatCurve=DataInterface() {propertyValues {Curve {keys [
          {arriveTangent=32.9812, interpMode=RCIM_Cubic, leaveTangent=32.9812, tangentMode=RCTM_User, time=0, value=0}
          {arriveTangentWeight=0.1389, interpMode=RCIM_Cubic, leaveTangentWeight=0.2778, tangentMode=RCTM_User, time=0.2, value=1}
         ]}}}
      Scale Curve=&FloatFromCurve(FloatCurve=DataInterface() {propertyValues {Curve {keys [<at 0 1 RCIM_Cubic>, {arriveTangent=-2.8655, interpMode=RCIM_Cubic, leaveTangent=-2.8655, tangentMode=RCTM_User, time=1, value=0}]}}})
     )))
   &GravityForce(Gravity=<0 0 -300>)
   &Drag(
    Drag=&MakeCustomFloatFromBool(Bool=$Particles.UseCollision, True Float=0.1, False Float=1)
    Use Linear Drag=true
    Use Rotational Drag=false
   )
   &SolveForcesAndVelocity(Clamp Velocity=false, Limit Acceleration=false)
   /SetVariables(Transient.VelocityScale:NiagaraFloat=&RemapRange(
     Input Value To Remap=&VectorLength(Vector=$Particles.Velocity)
     Input Min=75
     Input Max=125
     Output Min=0.25
     Output Max=1
    )) {isSetParameters=true, module=SetVariables_#4}
   &UpdateMeshOrientation(Rotation Rate=$Transient.VelocityScale, Rotation Vector=&RandomVector(Vector Scale=1))
   &Color(Color=&ColorFromCurve(
     ColorCurve=DataInterface() {propertyValues {
       AlphaCurve {keys [<at 0 1 RCIM_Cubic>]}
       BlueCurve {keys [<at 0 0.9943>, {arriveTangent=-0.9943, interpMode=RCIM_Cubic, leaveTangent=-0.9943, time=0.02, value=0}, <at 0.98 1>, <at 1 0 RCIM_Cubic>]}
       GreenCurve {keys [<at 0 0.9943>, {arriveTangent=0.0057, interpMode=RCIM_Cubic, leaveTangent=0.0057, time=0.02, value=0}, <at 0.98 0>, <at 1 1 RCIM_Cubic>]}
       RedCurve {keys [<at 0 0.9943>, {arriveTangent=-0.9943, interpMode=RCIM_Cubic, leaveTangent=-0.9943, time=0.02, value=1}, <at 0.98 0>, <at 1 0 RCIM_Cubic>]}
      }}
     CurveIndex=$Transient.VelocityScale
    ))
  ]
  rendererClasses [&NiagaraSpriteRendererProperties, &NiagaraMeshRendererProperties]
  renderers [
   {rendererClass=&NiagaraSpriteRendererProperties}
   {rendererClass=&NiagaraMeshRendererProperties, rendererIndex=1}
  ]
 }
 CausticGlows {
  bEnabled=false
  emitterUpdateScript [
   &EmitterState()
   &SpawnRate(SpawnRate=20, Use Spawn Probability=false)
  ]
  particleSpawnScript [
   /SetVariables(Particles.PickBone:NiagaraInt32=&RandomRangeInt(Minimum=1, Maximum=6)) {isSetParameters=true}
   /SetVariables(
    Particles.PickBoneMatch:NiagaraInt32=&SelectIntFromArray(Direct Array Index=$Particles.PickBone, Int Selection Array=DataInterface() {propertyValues {IntData [0,3,4,1,2,3,4]}})
    Particles.ScaleGroup:NiagaraInt32=&SelectIntFromArray(
     Direct Array Index=&ReturnExecIndex()
     Int Selection Array=DataInterface() {propertyValues {IntData [0,1,0,1,2,3,2,3,4,4]}}
    )
   ) {isSetParameters=true}
   &InitializeParticle(
    Color=<rgba 0.1565 0.6878 2 1>
    Lifetime=0.5
    Sprite Rotation Mode=::NewEnumerator2
    Sprite Size Mode=::NewEnumerator3
    Uniform Sprite Size=&SelectFloatFromArray(Direct Array Index=$Particles.ScaleGroup, Float Selection Array=DataInterface() {propertyValues {FloatData [70,60,70,60,50]}})
    UsePositionOffset=false
    Write Material Random=true
   )
   &ShapeLocation(
    MaskSpawnGroup=false
    Override Seed=false
    Shape Origin=&VectorFromFloat(Value=0)
    Sphere Radius=25
    UseOwnerScale=true
   )
   &SubUVAnimation(End Frame=3, Start Frame=2, SubUV Animation Mode:ENiagara_SubUVLookupMode=::NewEnumerator2)
   &AddVelocity(Velocity=<0 0 100>)
  ]
  particleUpdateScript [
   &ParticleState()
   &ScaleSpriteSize(Uniform Curve Sprite Scale=DataInterface() {propertyValues {Curve {keys [{interpMode=RCIM_Cubic, tangentMode=RCTM_User, time=0, value=0.5}, {interpMode=RCIM_Cubic, tangentMode=RCTM_User, time=1, value=1}]}}})
   &SpriteRotationRate(Rotation Rate=&FloatFromCurve(
     FloatCurve=DataInterface() {propertyValues {Curve {keys [{interpMode=RCIM_Cubic, tangentMode=RCTM_Break, time=0, value=1}, {interpMode=RCIM_Cubic, tangentMode=RCTM_Break, time=1, value=0.489}]}}}
     Scale Curve=&Lerp_Float(A=-150, B=150, Alpha=$Particles.Initial.MaterialRandom)
    ))
   &Drag(Drag=4, Use Linear Drag=true, Use Rotational Drag=true)
   &SolveForcesAndVelocity(Clamp Velocity=false, Limit Acceleration=false)
   &ScaleColor(
    Scale Alpha=&FloatFromCurve(
     FloatCurve=DataInterface() {propertyValues {Curve {keys [
         {interpMode=RCIM_Cubic, leaveTangentWeight=0.3534, tangentMode=RCTM_User, time=0, value=0}
         {interpMode=RCIM_Cubic, tangentMode=RCTM_User, time=0.75, value=1}
        ]}}}
     Scale Curve=&MakeCustomFloatFromBool(Bool=$User.IsFrontEnd, True Float=0.25, False Float=1)
    )
    ScaleA=true
    ScaleRGB=true
   )
   &DynamicMaterialParameters(
    Index 0 Param 1=&RandomRangeFloat(Minimum=0, Maximum=4)
    Index 0 Param 2=1
    Index 0 Param 3=1
    Index 0 Param 4=&FloatFromCurve(FloatCurve=DataInterface() {propertyValues {Curve {keys [{arriveTangent=-2.0044, interpMode=RCIM_Cubic, leaveTangent=-2.0044, tangentMode=RCTM_User, time=0, value=1}, <at 1 0.22>]}}})
    Index 3 Param 1=&RandomRangeFloat(Minimum=0, Maximum=1)
    Index 3 Param 2=&RandomRangeFloat(Minimum=0, Maximum=1)
    Index 3 Param 3=1
    Index 3 Param 4=1
    Module.Param0WriteEnabled=true
    Module.Param0WriteEnabled3=true
    Module.Param1WriteEnabled=true
    Module.Param1WriteEnabled3=true
    Module.Param2WriteEnabled=true
    Module.Param2WriteEnabled3=true
    Module.Param3WriteEnabled=true
    Module.Param3WriteEnabled3=true
    Write Parameter Index 3=true
   )
   &SkeletalMeshLocation(
    Bone / Socket Index=$Particles.PickBone
    Bone Sampling Mode=::NewEnumerator4
    MaskSpawnGroup=false
    Orientation Sampling=::NewEnumerator1
    Skeletal Mesh=$System.Skeletal Mesh
   )
   &SkeletalMeshLocation(
    Bone / Socket Index=$Particles.PickBoneMatch
    Bone Sampling Mode=::NewEnumerator4
    MaskSpawnGroup=false
    Orientation Sampling=::NewEnumerator1
    Position Sampling=::NewEnumerator2
    Skeletal Mesh=$System.Skeletal Mesh
   ) {module=SkeletalMeshLocation001}
   /SetVariables(Particles.GlowPosition:NiagaraPosition=&LerpPosition(
     A=$Particles.SkeletalMeshLocation.SampledPosition
     B=$Particles.SkeletalMeshLocation001.SampledPosition
     Alpha=&RandomRangeFloat(Minimum=0, Maximum=1)
    )) {isSetParameters=true}
  ]
  rendererClasses [&NiagaraSpriteRendererProperties]
  renderers [{rendererClass=&NiagaraSpriteRendererProperties}]
 }
 Decals {
  emitterUpdateScript [
   &EmitterState()
   &SpawnRate(SpawnRate=14, Use Spawn Probability=false)
  ]
  particleSpawnScript [
   &PartitionParticles(Number Of Partitions=7)
   /SetVariables(
    Particles.PickBone:NiagaraInt32=&Add_Integer(A=$Particles.PartitionParticles.Partition, B=1)
    Particles.ScaleGroup:NiagaraInt32=&SelectIntFromArray(Direct Array Index=$Particles.PartitionParticles.Partition, Int Selection Array=DataInterface() {propertyValues {IntData [0,0,1,1,2,2]}})
   ) {isSetParameters=true}
   /SetVariables(Transient.DecalScale:NiagaraFloat=&SelectFloatFromArray(Direct Array Index=$Particles.ScaleGroup, Float Selection Array=DataInterface() {propertyValues {FloatData [40,35,30]}})) {isSetParameters=true}
   &InitializeParticle(
    Color=&HueShiftLinearColor(
     Hue=-0.2
     Saturation=1
     Value=1
     Linear Color=&RandomRangeLinearColor(Minimum=<rgba 0.113 0.1994 1 1>, Maximum=<rgba 0.1913 0.3713 1 1>)
    )
    Position Offset=<0 -30 110>
    Sprite Rotation Mode=::NewEnumerator2
    Sprite Size Mode=::NewEnumerator3
    Uniform Sprite Size=&Multiply_Float(
     A=$Transient.DecalScale
     B=&MakeCustomFloatFromBool(Bool=$User.IsLargeBody, True Float=1.35, False Float=1)
    )
    UsePositionOffset=false
    Write Material Random=true
   )
  ]
  particleUpdateScript [
   &ParticleState()
   &SpriteRotationRate(Rotation Rate=&Multiply_Float(
     A=&RandomRangeFloat(Minimum=10, Maximum=30)
     B=&UniformAOrBFloat(A=-1, Evaluation Type=::NewEnumerator0, Recalculate Random Each Loop=true, Override Seed=false)
    ))
   /SetVariables(Transient.ToDMult:NiagaraFloat=&FNTimeOfDayScalar(
     Night Scale=1.5
     Morning Scale=1.2
     Day Scale=&MakeCustomFloatFromBool(Bool=$User.IsFrontEnd, True Float=3, False Float=3)
     Evening Scale=0.9
    )) {isSetParameters=true}
   &ScaleColor(
    Scale Alpha=&FloatFromCurve(
     FloatCurve=DataInterface() {propertyValues {Curve {keys [
         {interpMode=RCIM_Cubic, tangentMode=RCTM_User, time=0, value=0}
         {interpMode=RCIM_Cubic, tangentMode=RCTM_User, time=0.5, value=1}
         {tangentMode=RCTM_User, time=1, value=0}
        ]}}}
     Scale Curve=&MakeCustomFloatFromBool(Bool=$User.IsLargeBody, True Float=1, False Float=1)
    )
    Scale RGB=&VectorFromFloat(Value=&Multiply_Float(
      A=$Transient.ToDMult
      B=&Lerp_Float(A=1, B=0.5, Alpha=&Multiply_FloatByInt(Float=0.5, Integer=$Particles.ScaleGroup))
     ))
    ScaleA=true
    ScaleRGB=true
   )
   &DynamicMaterialParameters(
    Index 0 Param 1=&MakeCustomFloatFromBool(Bool=$User.IsFrontEnd, True Float=2.5, False Float=2)
    Index 0 Param 2=&RandomRangeFloat(Minimum=0, Maximum=1)
    Index 0 Param 3=&RandomRangeFloat(Minimum=0, Maximum=1)
    Index 0 Param 4=&MakeCustomFloatFromBool(Bool=$User.IsFrontEnd, True Float=0.6, False Float=3)
    Module.Index0Param0Write=true
    Module.Index0Param1Write=true
    Module.Index0Param2Write=true
    Module.Index0Param3Write=true
   )
   &SkeletalMeshLocation(
    Bone / Socket Index=$Particles.PickBone
    Bone Sampling Mode=::NewEnumerator4
    MaskSpawnGroup=false
    Orientation Sampling=::NewEnumerator1
    Skeletal Mesh=$System.Skeletal Mesh
   )
   &CameraOffset(Camera Offset Amount=&MakeCustomFloatFromBool(Bool=$User.IsLargeBody, True Float=8, False Float=6))
  ]
  rendererClasses [&NiagaraSpriteRendererProperties]
  renderers [
   {rendererClass=&NiagaraSpriteRendererProperties}
   {rendererClass=&NiagaraSpriteRendererProperties, rendererIndex=1}
  ]
 }
 Drips {
  emitterUpdateScript [
   &EmitterState()
   &SpawnRate(
    SpawnRate=&MakeCustomFloatFromBool(Bool=$User.IsFrontEnd, True Float=45, False Float=80)
    Use Spawn Probability=false
   )
  ]
  particleSpawnScript [
   /SetVariables(Particles.PickBone:NiagaraInt32=&RandomRangeInt(Minimum=3, Maximum=6)) {isSetParameters=true}
   /SetVariables(Particles.PickBoneMatch:NiagaraInt32=&SelectIntFromArray(Direct Array Index=$Particles.PickBone, Int Selection Array=DataInterface() {propertyValues {IntData [0,0,0,1,2,3,4]}})) {isSetParameters=true}
   &SkeletalMeshLocation(
    Bone / Socket Index=$Particles.PickBone
    Bone Sampling Mode=::NewEnumerator4
    MaskSpawnGroup=false
    Orientation Sampling=::NewEnumerator1
    Position Sampling=::NewEnumerator2
    Sampled Position Offset=&RandomVector(Vector Scale=5)
    Sampled Position Offset Coordinate Space:ENiagaraSkelMeshTransforms=::NewEnumerator1
    Skeletal Mesh=$System.Skeletal Mesh
   )
   &SkeletalMeshLocation(
    Bone / Socket Index=$Particles.PickBoneMatch
    Bone Sampling Mode=::NewEnumerator4
    MaskSpawnGroup=false
    Orientation Sampling=::NewEnumerator1
    Position Sampling=::NewEnumerator2
    Sampled Position Offset=&RandomVector(Vector Scale=5)
    Sampled Position Offset Coordinate Space:ENiagaraSkelMeshTransforms=::NewEnumerator1
    Skeletal Mesh=$System.Skeletal Mesh
   ) {module=SkeletalMeshLocation001}
   &SkeletalMeshLocation(
    Bone / Socket Index=7
    Bone Sampling Mode=::NewEnumerator4
    MaskSpawnGroup=false
    Orientation Sampling=::NewEnumerator1
    Position Sampling=::NewEnumerator2
    Skeletal Mesh=$System.Skeletal Mesh
   ) {module=SkeletalMeshLocation002}
   /SetVariables(Transient.VelocityModifier:NiagaraFloat=&RemapRange(
     Input Value To Remap=&VectorLength(Vector=$Particles.SkeletalMeshLocation.SampledVelocity)
     Input Min=15
     Input Max=100
     Output Min=0
     Output Max=1
    )) {isSetParameters=true}
   &InitializeParticle(
    Color=<rgba 0.669 0.8256 1 0.4>
    Lifetime=&Lerp_Float(
     A=0.25
     B=0.75
     Alpha=&FloatFromCurve(
      FloatCurve=DataInterface() {propertyValues {Curve {keys [{tangentMode=RCTM_User, time=0, value=0}, <at 0.75 0.125>, {tangentMode=RCTM_User, time=1, value=1}]}}}
      CurveIndex=&RandomRangeFloat(Minimum=0, Maximum=1)
     )
    )
    Mass Max=1.2
    Mass Min=0.6
    Mass Mode=::NewEnumerator3
    Position=&LerpPosition(
     A=$Particles.SkeletalMeshLocation.SampledPosition
     B=$Particles.SkeletalMeshLocation001.SampledPosition
     Alpha=&RandomRangeFloat(Minimum=0, Maximum=1)
    )
    Position Mode=::NewEnumerator1
    Sprite Rotation Mode=::NewEnumerator2
    Sprite Size=&Lerp_Vector2(A=<1.5 4>, B=<2.5 13.5>, Alpha=&RandomRangeFloat(Minimum=0, Maximum=1))
    Sprite Size Mode=::NewEnumerator4
    UsePositionOffset=false
    Write Material Random=true
   )
   &InheritVelocity(
    Inherited Velocity Amount Scale=<1 1 0.2>
    Inherited Velocity Speed Limit=500
    Speed Limit=true
    UseSpeedThreshold=false
    Velocity Source=$Particles.SkeletalMeshLocation.SampledVelocity
   ) {enabled=false}
   &AddVelocity(
    Rotation Coordinate Space=::World
    Velocity=&Add_Vector(
     A=&RandomRangeVector(Minimum=<-3 -3 15>, Maximum=<3 3 50>)
     B=&Multiply_VectorByFloat(
      Vector=&NormalizeVector(
       Vector To Normalize=&Subtract_Position(A=$Particles.Position, B=$Particles.SkeletalMeshLocation002.SampledPosition)
       Safe Normalize=true
       Safe Fallback Threshold=0
       Safe Fallback Vector=<0 0 0.001>
      )
      Float=&RandomRangeFloat(Minimum=100, Maximum=200)
     )
    )
    Velocity Speed Scale=&Lerp_Float(A=1, B=0.5, Alpha=$Transient.VelocityModifier)
   )
   /SetVariables(Particles.Color=&ColorFromCurve(
     ColorCurve=DataInterface() {propertyValues {
       AlphaCurve {keys [<at 0 1 RCIM_Cubic>]}
       BlueCurve {keys [<at 0 1 RCIM_Cubic>,<at 0.02 0>,<at 0.98 1>,<at 1 0 RCIM_Cubic>]}
       GreenCurve {keys [<at 0 1 RCIM_Cubic>,<at 0.02 0>,<at 0.98 0>,<at 1 1 RCIM_Cubic>]}
       RedCurve {keys [<at 0 1 RCIM_Cubic>,<at 0.02 1>,<at 0.98 0>,<at 1 0 RCIM_Cubic>]}
      }}
     CurveIndex=$Transient.VelocityModifier
    )) {enabled=false, isSetParameters=true}
   &DynamicMaterialParameters(
    Index 0 Param 1=&UniformRangedFloat(Minimum=4, Maximum=10)
    Module.Param0WriteEnabled=true
    Module.Param1WriteEnabled=false
    Module.Param2WriteEnabled=false
    Module.Param3WriteEnabled=false
   )
   &KillParticles(Kill Particles=&SetBoolByFloatComparison(A=$Transient.VelocityModifier, B=0.5, Comparison Type=::NewEnumerator7))
  ]
  particleUpdateScript [
   &ParticleState()
   &ScaleSpriteSize(
    Non-Uniform Curve Sprite Scale=DataInterface() {propertyValues {
      XCurve {keys [<at 0.5 1 RCIM_Cubic>, {arriveTangent=-0.4016, tangentMode=RCTM_User, time=1, value=0.25}]}
      YCurve {keys [
        {arriveTangent=3.3463, interpMode=RCIM_Cubic, leaveTangent=3.3463, tangentMode=RCTM_User, time=0, value=0}
        <at 0.3 1>
        {arriveTangent=-1.281, interpMode=RCIM_Cubic, leaveTangent=-1.281, tangentMode=RCTM_User, time=0.5, value=1}
        {arriveTangent=-0.1464, tangentMode=RCTM_User, time=1, value=0}
       ]}
     }}
    Scale Sprite Size Mode=::NewEnumerator3
   )
   &Drag(Use Linear Drag=true, Use Rotational Drag=false)
   &GravityForce()
   &SolveForcesAndVelocity(Clamp Velocity=false, Limit Acceleration=false)
  ]
  rendererClasses [&NiagaraSpriteRendererProperties]
  renderers [
   {rendererClass=&NiagaraSpriteRendererProperties}
   {rendererClass=&NiagaraSpriteRendererProperties, rendererIndex=1}
  ]
 }
 SplashSprites {
  emitterUpdateScript [
   &EmitterState()
   &SpawnRate(SpawnRate=30, Use Spawn Probability=false)
   &SpawnRate(SpawnGroup=1, SpawnRate=8, Use Spawn Probability=false) {module=SpawnRate001}
  ]
  particleSpawnScript [
   /SetVariables(Particles.Variance:NiagaraFloat=&RandomRangeFloat(Minimum=0, Maximum=1)) {isSetParameters=true}
   /SetVariables(Particles.PickBone:NiagaraInt32=&RandomRangeInt(Minimum=1, Maximum=4)) {isSetParameters=true}
   /SetVariables(Particles.PickBoneMatch:NiagaraInt32=&SelectIntFromArray(Direct Array Index=$Particles.PickBone, Int Selection Array=DataInterface() {propertyValues {IntData [0,3,4,1,2,3,4]}})) {isSetParameters=true}
   &SkeletalMeshLocation(
    Bone / Socket Index=$Particles.PickBone
    Bone Sampling Mode=::NewEnumerator4
    MaskSpawnGroup=false
    Orientation Sampling=::NewEnumerator1
    Position Sampling=::NewEnumerator2
    Sampled Position Offset=&RandomVector(Vector Scale=5)
    Sampled Position Offset Coordinate Space:ENiagaraSkelMeshTransforms=::NewEnumerator1
    Skeletal Mesh=$System.Skeletal Mesh
   )
   &SkeletalMeshLocation(
    Bone / Socket Index=$Particles.PickBoneMatch
    Bone Sampling Mode=::NewEnumerator4
    MaskSpawnGroup=false
    Orientation Sampling=::NewEnumerator1
    Position Sampling=::NewEnumerator2
    Sampled Position Offset=&RandomVector(Vector Scale=5)
    Sampled Position Offset Coordinate Space:ENiagaraSkelMeshTransforms=::NewEnumerator1
    Skeletal Mesh=$System.Skeletal Mesh
   ) {module=SkeletalMeshLocation001}
   &SkeletalMeshLocation(
    Bone / Socket Index=7
    Bone Sampling Mode=::NewEnumerator4
    MaskSpawnGroup=false
    Orientation Sampling=::NewEnumerator1
    Position Sampling=::NewEnumerator2
    Skeletal Mesh=$System.Skeletal Mesh
   ) {module=SkeletalMeshLocation002}
   &InitializeParticle(
    Color=&MakeCustomLinearColorFromBool(
     Bool=$User.IsFrontEnd
     True Color=&RandomRangeLinearColor(
      Minimum=<rgba 0.1548 0.1548 1 1>
      Maximum=&MaskLinearColorBySpawnGroup(Masked Linear Color=<rgba 0.3843 0.6767 1 1>, Passthrough Linear Color=<rgba 0.3843 0.487 1 1>)
     )
     False Color=&RandomRangeLinearColor(
      Minimum=<rgba 0.1548 0.4365 1 1>
      Maximum=&MaskLinearColorBySpawnGroup(Masked Linear Color=<rgba 0.3843 0.8974 1 1>, Passthrough Linear Color=<rgba 0.3843 0.6922 1 1>)
     )
    )
    Lifetime Max=0.75
    Lifetime Min=0.5
    Lifetime Mode=::NewEnumerator1
    Position=&MaskVectorBySpawnGroup(
     Masked Vector=&LerpPosition(
      A=$Particles.SkeletalMeshLocation.SampledPosition
      B=$Particles.SkeletalMeshLocation001.SampledPosition
      Alpha=&RandomRangeFloat(Minimum=0, Maximum=1)
     )
     Passthrough Vector=$Particles.SkeletalMeshLocation002.SampledPosition
    )
    Position Mode=::NewEnumerator1
    Sprite Rotation Mode=::NewEnumerator2
    Sprite Size Mode=::NewEnumerator3
    Sprite UV Mode=::NewEnumerator3
    Uniform Sprite Size=&MaskFloatBySpawnGroup(
     Masked Float=&Lerp_Float(A=18, B=40, Alpha=$Particles.Variance)
     Passthrough Float=&Lerp_Float(A=90, B=120, Alpha=$Particles.Variance)
    )
    UsePositionOffset=false
    Write Material Random=true
   )
   &AddVelocity(
    Velocity=&MaskVectorBySpawnGroup(
     Masked Vector=&Add_Vector(
      A=&RandomRangeVector(Minimum=<0 0 25>, Maximum=<0 0 100>)
      B=&Multiply_VectorByFloat(
       Vector=&NormalizeVector(
        Vector To Normalize=&Subtract_Position(A=$Particles.Position, B=$Particles.SkeletalMeshLocation002.SampledPosition)
        Safe Normalize=true
        Safe Fallback Threshold=0
        Safe Fallback Vector=<0 0 0.001>
       )
       Float=&Lerp_Float(A=50, B=100, Alpha=$Particles.Variance)
      )
     )
     Passthrough Vector=&RandomRangeVector(Minimum=<0 0 50>, Maximum=<0 0 100>)
    )
    Velocity Speed Scale=&Lerp_Float(A=1.2, B=0.9, Alpha=$Particles.Variance)
   )
   &CameraOffset(Camera Offset Amount=&MaskFloatBySpawnGroup(
     Masked Float=0
     Passthrough Float=&MakeCustomFloatFromBool(Bool=$User.IsLargeBody, True Float=-53, False Float=-33)
    ))
   /SetVariables(Particles.SubImageIndex=&RandomRangeFloat(Minimum=0, Maximum=3)) {isSetParameters=true}
  ]
  particleUpdateScript [
   &ParticleState()
   &GravityForce(Gravity=<0 0 -500>)
   &SolveForcesAndVelocity(Clamp Velocity=false, Limit Acceleration=false)
   &ScaleSpriteSize(Uniform Curve Sprite Scale=DataInterface() {propertyValues {Curve {keys [
        {arriveTangent=2.1662, interpMode=RCIM_Cubic, leaveTangent=2.1662, tangentMode=RCTM_User, time=0, value=0.1}
        {interpMode=RCIM_Cubic, tangentMode=RCTM_User, time=1, value=1}
       ]}}})
   &ScaleColor(
    Scale Alpha=&FloatFromCurve(FloatCurve=DataInterface() {propertyValues {Curve {keys [{interpMode=RCIM_Cubic, tangentMode=RCTM_User, time=0, value=1}, {interpMode=RCIM_Cubic, tangentMode=RCTM_User, time=1, value=0}]}}})
    Scale RGB=&MakeVectorFromLinearColorRGB(Color=&ColorFromCurve(ColorCurve=DataInterface() {propertyValues {
        AlphaCurve {keys [<at 0 1 RCIM_Cubic>,<at 1 1 RCIM_Cubic>]}
        BlueCurve {keys [<at 0.2219 0.9792>,<at 0.6021 0.2245>,<at 1 0.2969>]}
        GreenCurve {keys [<at 0.2219 0.5638>,<at 0.6021 0.164>,<at 1 0.0648>]}
        RedCurve {keys [<at 0.2219 0.0804>,<at 0.6021 0>,<at 1 0.0322>]}
       }}))
    ScaleA=true
    ScaleRGB=true
   )
   &DynamicMaterialParameters(
    Index 0 Param 1=&RandomRangeFloat(Minimum=0, Maximum=1)
    Index 0 Param 2=&FloatFromCurve(
     FloatCurve=DataInterface() {propertyValues {Curve {keys [<at 0.1 1 RCIM_Cubic>, {interpMode=RCIM_Cubic, tangentMode=RCTM_User, time=0.3, value=0}]}}}
     Scale Curve=0
    )
    Index 0 Param 3=&MaskFloatBySpawnGroup(
     Masked Float=&Lerp_Float(
      A=1
      B=4
      Alpha=&FloatFromCurve(FloatCurve=DataInterface() {propertyValues {Curve {keys [<at 0 1 RCIM_Cubic>,<at 0.5 0 RCIM_Cubic>]}}})
     )
     Passthrough Float=&Lerp_Float(
      A=0
      B=3
      Alpha=&FloatFromCurve(FloatCurve=DataInterface() {propertyValues {Curve {keys [
           {interpMode=RCIM_Cubic, tangentMode=RCTM_User, time=0, value=0}
           {arriveTangent=0.3092, interpMode=RCIM_Cubic, leaveTangent=0.3092, tangentMode=RCTM_User, time=1, value=1}
          ]}}})
     )
    )
    Index 0 Param 4=&MaskFloatBySpawnGroup(
     Masked Float=&FloatFromCurve(FloatCurve=DataInterface() {propertyValues {Curve {keys [{interpMode=RCIM_Cubic, leaveTangent=4.137, tangentMode=RCTM_Break, time=0.1, value=0}, <at 1 1>]}}})
     Passthrough Float=&FloatFromCurve(FloatCurve=DataInterface() {propertyValues {Curve {keys [{interpMode=RCIM_Cubic, leaveTangent=1.9612, leaveTangentWeight=1.2768, tangentMode=RCTM_Break, time=0.1, value=0}, <at 1 1>]}}})
    )
    Module.Index0Param0Write=true
    Module.Index0Param1Write=true
    Module.Index0Param2Write=true
    Module.Index0Param3Write=true
   )
   /SetVariables(Particles.Color=&MaskLinearColorBySpawnGroup(Masked Linear Color=$Particles.Color, Passthrough Linear Color=$Particles.Initial.Color)) {isSetParameters=true}
  ]
  rendererClasses [&NiagaraSpriteRendererProperties]
  renderers [{rendererClass=&NiagaraSpriteRendererProperties}]
 }
 Splatter {
  emitterUpdateScript [
   &EmitterState()
   &SpawnRate(SpawnRate=10, Use Spawn Probability=false)
   &SpawnRate(SpawnGroup=1, SpawnRate=10, Use Spawn Probability=false) {module=SpawnRate001}
  ]
  particleSpawnScript [
   /SetVariables(
    Particles.LifetimeScale:NiagaraFloat=&RandomRangeFloat(Minimum=0, Maximum=1)
    Particles.PartitionParticles.Partition:NiagaraInt32=&SelectIntFromArray(
     Array Sampling Mode=::NewEnumerator1
     Direct Array Index=?
     Int Selection Array=DataInterface() {propertyValues {IntData [0,0,1,2,5]}}
     Random Seed=0
    )
    Particles.VisibilityTag=&SpawnGroup()
   ) {isSetParameters=true}
   /SetVariables(
    Particles.PivotOffset:Vector2f=&SelectVector2DFromArray(
     Direct Array Index=$Particles.PartitionParticles.Partition
     Vector 2D Selection Array=DataInterface() {propertyValues {FloatData [<0.5 0.6>, <0.5 1.25>, <0.5 1>, <0.2 0.8>, <0.5 0.2>, <0.5 0.35>]}}
    )
    Particles.RotationPicker:NiagaraFloat=&RandomRangeFloat(Minimum=0, Maximum=1)
   ) {isSetParameters=true}
   &InitializeParticle(
    Color Maximum=<rgba 0.0991 0.4877 1 1>
    Color Minimum=<rgba 0.0574 0.2279 1 1>
    Color Mode=::NewEnumerator2
    Lifetime=&Lerp_Float(A=0.6, B=0.9, Alpha=$Particles.LifetimeScale)
    Sprite Rotation Angle=&Add_Float(
     A=&Lerp_Float(A=-0.2, B=0.2, Alpha=$Particles.RotationPicker)
     B=&SelectFloatFromArray(
      Direct Array Index=$Particles.PartitionParticles.Partition
      Float Selection Array=DataInterface() {propertyValues {FloatData [0,0,0,0,0.5,0.5]}}
     )
    )
    Sprite Rotation Mode=::NewEnumerator1
    Sprite Size=&Multiply_Vector2D_ByFloat(
     Float=&Multiply_Float(
      A=&MakeCustomFloatFromBool(Bool=$User.IsLargeBody, True Float=1.6, False Float=1.2)
      B=&MaskFloatBySpawnGroup(Masked Float=&RandomRangeFloat(Minimum=0.8, Maximum=1.1), Passthrough Float=0.5)
     )
     Vector2D=&Multiply_Vector2D_ByFloat(
      Float=&SelectFloatFromArray(
       Direct Array Index=$Particles.PartitionParticles.Partition
       Float Selection Array=DataInterface() {propertyValues {FloatData [110,50,60,0,0,100]}}
      )
      Vector2D=<0.75 1>
     )
    )
    Sprite Size Mode=::NewEnumerator4
    Sprite UV Mode=::NewEnumerator1
    UsePositionOffset=true
    Write Material Random=true
   )
   &DynamicMaterialParameters(
    Index 1 Param 1=&MakeFloatFromInt(Int=$Particles.PartitionParticles.Partition)
    Module.Index1Param0Write=true
    Module.Index1Param1Write=false
    Module.Index1Param2Write=true
    Module.Index1Param3Write=true
    Write Parameter Index 0=false
    Write Parameter Index 1=true
   ) {module=DynamicMaterialParameters001}
   &CameraOffset(Camera Offset Amount=&MaskFloatBySpawnGroup(
     Masked Float=&MakeCustomFloatFromBool(Bool=$User.IsLargeBody, True Float=-55, False Float=-35)
     Passthrough Float=&MakeCustomFloatFromBool(Bool=$User.IsLargeBody, True Float=-50, False Float=-30)
    ))
   /SetVariables(Particles.Color=&MaskLinearColorBySpawnGroup(
     Masked Linear Color=$Particles.Color
     Passthrough Linear Color=&LerpLinearColors(
      StartColor=<rgba 1 1 1 1>
      EndColor=$Particles.Color
      LerpFactor=&RandomRangeFloat(Minimum=0.5, Maximum=0.75)
     )
    )) {isSetParameters=true}
  ]
  particleUpdateScript [
   &ParticleState()
   &SpriteRotationRate(Rotation Rate=&Multiply_Float(
     A=&Lerp_Float(A=-50, B=50, Alpha=$Particles.RotationPicker)
     B=&Lerp_Float(A=1, B=1.25, Alpha=$Particles.LifetimeScale)
    ))
   &ScaleSpriteSize(Uniform Curve Sprite Scale=DataInterface() {propertyValues {Curve {keys [
        {arriveTangent=2.6348, interpMode=RCIM_Cubic, leaveTangent=2.6348, tangentMode=RCTM_User, time=0, value=0}
        {tangentMode=RCTM_User, time=1, value=1}
       ]}}})
   &ScaleColor(
    Scale RGB=&Multiply_VectorByFloat(
     Vector=<1 1 1>
     Float=&FNTimeOfDayScalar(
      Night Scale=1.3
      Morning Scale=1.2
      Day Scale=&MakeCustomFloatFromBool(Bool=$User.IsFrontEnd, True Float=0.8, False Float=1)
      Evening Scale=1.3
     )
    )
    ScaleA=true
    ScaleRGB=true
   )
   &DynamicMaterialParameters(
    Index 0 Param 2=&FloatFromCurve(FloatCurve=DataInterface() {propertyValues {Curve {keys [
         <at 0 1 RCIM_Cubic>
         {arriveTangent=-0.0097, interpMode=RCIM_Cubic, leaveTangent=-0.0097, tangentMode=RCTM_User, time=0.4, value=0}
         <at 1 1 RCIM_Cubic>
        ]}}})
    Index 1 Param 2=1
    Index 1 Param 4=&Multiply_Float(
     A=&SelectFloatFromArray(
      Array Sampling Mode=::NewEnumerator0
      Direct Array Index=?
      Float Selection Array=DataInterface() {propertyValues {FloatData [1,1,1,1,1]}}
      Random Seed=0
     )
     B=&RandomRangeFloat(Minimum=0.95, Maximum=1.05)
    )
    Module.Index0Param0Write=true
    Module.Index0Param1Write=true
    Module.Index0Param2Write=true
    Module.Index0Param3Write=true
    Module.Index1Param0Write=false
    Module.Index1Param1Write=true
    Module.Index1Param2Write=false
    Module.Index1Param3Write=true
    Write Parameter Index 1=true
   )
   &SkeletalMeshLocation(
    Bone / Socket Index=7
    Bone Sampling Mode=::NewEnumerator4
    MaskSpawnGroup=false
    Position Sampling=::NewEnumerator1
    Sampled Position Offset=&Add_Vector(
     A=&RandomRangeVector(Minimum=<0 0 -10>, Maximum=<0 0 10>)
     B=&Lerp_Vector(A=<0 0 0>, B=<0 0 0>, Alpha=$Particles.RotationPicker)
    )
    Skeletal Mesh=$System.Skeletal Mesh
   ) {module=SkeletalMeshLocation001}
   &GravityForce(Gravity=&RandomRangeVector(Minimum=<0 0 -40>, Maximum=<0 0 -60>))
   &SolveForcesAndVelocity(Clamp Velocity=false, Limit Acceleration=false)
  ]
  rendererClasses [&NiagaraSpriteRendererProperties]
  renderers [
   {rendererClass=&NiagaraSpriteRendererProperties}
   {rendererClass=&NiagaraSpriteRendererProperties, rendererIndex=1}
  ]
 }
}
summary {
 systemName=NS_Aura_Splash
 emitters [
  {emitterName=Drips, rendererClasses [&NiagaraSpriteRendererProperties]}
  {emitterName=Beachballs, rendererClasses [&NiagaraSpriteRendererProperties, &NiagaraMeshRendererProperties]}
  {emitterName=Decals, rendererClasses [&NiagaraSpriteRendererProperties]}
  {bEnabled=false, emitterName=CausticGlows, rendererClasses [&NiagaraSpriteRendererProperties]}
  {emitterName=Splatter, rendererClasses [&NiagaraSpriteRendererProperties]}
  {emitterName=SplashSprites, rendererClasses [&NiagaraSpriteRendererProperties]}
  {
   bEnabled=false
   emitterName=Beachballs001
   rendererClasses [&NiagaraSpriteRendererProperties, &NiagaraMeshRendererProperties]
  }
 ]
 userVariables [
  User.bIsFrontEnd {type=NiagaraBool}
  User.bIsFrontEndPreview {type=NiagaraBool}
  User.IsFrontEnd {type=NiagaraBool}
  User.IsLargeBody {type=NiagaraBool}
 ]
}
